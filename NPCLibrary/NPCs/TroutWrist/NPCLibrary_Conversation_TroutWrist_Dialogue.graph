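# Conversation database for the TroutWrist NPC (Peelverine), providing the dialogue tree reached through its Dialogue entry tag.
# An "Ask about..." speech node branches to a Peelverine flavor topic, a BACK option, and a link to a remote Season31 Found AskAbout entry.
# The Peelverine topic plays a linear chain of stage-direction and spoken lines about being wolverine-fierce yet banana-soft, then offers BACK to return to the start.
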ConversationDatabase /NPCLibrary/NPCs/TroutWrist/NPCLibrary_Conversation_TroutWrist_Dialogue
entry_points [{tag=Conversation.Entry.TroutWrist.Dialogue, destinations [#1]}]
nodes [
 ConversationEntryPointNode_0 {
  id=#1
  class=ConversationEntryPointNode
  outputs [#2]
  properties {entryTag {tagName=Conversation.Entry.TroutWrist.Dialogue}}
 }
 Speech_C_0 {
  id=#2
  class=Speech_C
  outputs [#3, #4, #5]
  properties {nodeName="", softTaskIcon=&T_UI_NPC_question, generalConfig {defaultMessage=(Peelverine squints and nods.)}}
  sub_nodes [ConversationChoiceNode_5 {class=ConversationChoiceNode, properties {defaultChoiceDisplayText=Ask about...}}]
 }
 ConversationLinkNode_0 {
  id=#5
  class=ConversationLinkNode
  properties {remoteEntryTag {tagName=Conversation.Entry.Season31.Found.TroutWrist.AskAbout}}
 }
 Back_C_0 {
  id=#4
  class=Back_C
  sub_nodes [ConversationChoiceNode_7 {class=ConversationChoiceNode, properties {defaultChoiceDisplayText=BACK}}]
 }
 Speech_C_7 {
  id=#3
  class=Speech_C
  outputs [#6]
  properties {
   nodeName=""
   softTaskIcon=&T_UI_NPC_question
   generalConfig {defaultMessage=(Peelverine looks off into the distance.)}
  }
  sub_nodes [ConversationChoiceNode_15 {class=ConversationChoiceNode, properties {defaultChoiceDisplayText=Peelverine}}]
 }
 Speech_C_15 {
  id=#6
  class=Speech_C
  outputs [#7]
  properties {
   nodeName=""
   softTaskIcon=&T_UI_NPC_question
   generalConfig {defaultMessage=(He sighs, then pats himself on the peel.)}
  }
 }
 Speech_C_14 {
  id=#7
  class=Speech_C
  outputs [#8]
  properties {
   nodeName=""
   softTaskIcon=&T_UI_NPC_question
   generalConfig {defaultMessage=It's hard knowing you're as ferocious and wild as a wolverine.}
  }
 }
 Speech_C_1 {
  id=#8
  class=Speech_C
  outputs [#9]
  properties {
   nodeName=""
   softTaskIcon=&T_UI_NPC_question
   generalConfig {defaultMessage=Yet as soft and delicious as a banana.}
  }
 }
 ReturnToStart_C_23 {
  id=#9
  class=ReturnToStart_C
  properties {softTaskIcon=&T_UI_NPC_Back}
  sub_nodes [ConversationChoiceNode_32 {class=ConversationChoiceNode, properties {defaultChoiceDisplayText=BACK}}]
 }
]
